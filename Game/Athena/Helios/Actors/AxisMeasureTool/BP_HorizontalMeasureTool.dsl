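; BP_HorizontalMeasureTool is an actor that displays a horizontal measuring bar of configurable length.
; Its construction script scales the Cube along X, reading Length as floor units, raw units or hundreds of units according to MeasureType, which reuses the vertical tool's enum.
; It then snaps BillboardTop to the Cube's "end" socket and sets its visibility from ShowBillboardIcons.
; BeginPlay, Tick and overlap events are empty.
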(event EventBeginPlay)

(event Collision|EventActorBeginOverlap
  (params (OtherActor :type Actor%20Object%20Reference)))

(event EventTick
  (params (DeltaSeconds :type Float%20%28single-precision%29)))

(fn ConstructionScript
  (Rendering|SetVisibility (Variables|Default|GetBillboardTop) (Variables|Default|GetShowBillboardIcons) false)
  (switch Utilities|FlowControl|Switch|SwitchonEVerticalMeasureToolTypes (Variables|Default|GetMeasureType)
    (:NewEnumerator0
      (Transformation|SetWorldScale3D (Variables|Default|GetCube) (Math|Vector|MakeVectorNetQuantize (* (Variables|Default|GetLength) (LocalVariables|GetFloorUnits)) 256.0 256.0))
      (Transformation|SetWorldLocation (Variables|Default|GetBillboardTop) (Transformation|GetSocketLocation (Variables|Default|GetCube) "end") false false))
    (:NewEnumerator1
      (Transformation|SetWorldScale3D (Variables|Default|GetCube) (Math|Vector|MakeVector (Math|Conversions|ToFloat(Integer) (Variables|Default|GetLength)) 256.0 256.0))
      (Transformation|SetWorldLocation (Variables|Default|GetBillboardTop) (Transformation|GetSocketLocation (Variables|Default|GetCube) "end") false false))
    (:NewEnumerator2
      (Transformation|SetWorldScale3D (Variables|Default|GetCube) (Math|Vector|MakeVectorNetQuantize (* (Math|Conversions|ToFloat(Integer) (Variables|Default|GetLength)) 1e+02) 256.0 256.0))
      (Transformation|SetWorldLocation (Variables|Default|GetBillboardTop) (Transformation|GetSocketLocation (Variables|Default|GetCube) "end") false false))))
