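; Placeable vertical measuring actor that sizes a cube to a chosen Height, apparently as a level-building reference.
; The construction script sets top billboard visibility and can set a vertical snap grid size for builder-grid alignment.
; A measure-type switch scales the cube's Z by Height in floor units, raw units or hundreds of units, then moves the billboard to the cube's "end" socket.
; BeginPlay, overlap and Tick events are empty.
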(event EventBeginPlay)

(event Collision|EventActorBeginOverlap
  (params (OtherActor :type Actor%20Object%20Reference)))

(event EventTick
  (params (DeltaSeconds :type Float%20%28single-precision%29)))

(fn ConstructionScript
  (Rendering|SetVisibility (Variables|Default|GetBillboardTop) (Variables|Default|GetShowBillboardIcons) false)
  (if (Variables|Default|GetShouldSnapToBuilderGrid)
    (Building|SetVertSnapGridSize 384.0))
  (switch Utilities|FlowControl|Switch|SwitchonEVerticalMeasureToolTypes (Variables|Default|GetMeasureType)
    (:NewEnumerator0
      (Transformation|SetWorldScale3D (Variables|Default|GetCube) (Math|Vector|MakeVectorNetQuantize 256.0 256.0 (* (Variables|Default|GetHeight) (LocalVariables|GetFloorUnits))))
      (Transformation|SetWorldLocation (Variables|Default|GetBillboardTop) (Transformation|GetSocketLocation (Variables|Default|GetCube) "end") false false))
    (:NewEnumerator1
      (Transformation|SetWorldScale3D (Variables|Default|GetCube) (Math|Vector|MakeVectorNetQuantize 256.0 256.0 (Math|Conversions|ToFloat(Integer) (Variables|Default|GetHeight))))
      (Transformation|SetWorldLocation (Variables|Default|GetBillboardTop) (Transformation|GetSocketLocation (Variables|Default|GetCube) "end") false false))
    (:NewEnumerator2
      (Transformation|SetWorldScale3D (Variables|Default|GetCube) (Math|Vector|MakeVectorNetQuantize 256.0 256.0 (* (Math|Conversions|ToFloat(Integer) (Variables|Default|GetHeight)) 1e+02)))
      (Transformation|SetWorldLocation (Variables|Default|GetBillboardTop) (Transformation|GetSocketLocation (Variables|Default|GetCube) "end") false false))))
